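; Blueprint interface for volume objects shown in the MusicEncounter dashboard tool.
; It declares two query functions: one reporting whether the listener is inside the volume, and one returning the volume's delegates.
; Both are signature stubs with default returns, so implementing classes supply the behavior.
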(fn GetIsListenerinVolume
  (return false))

(fn GetVolumeDelegates
  (return 0))
